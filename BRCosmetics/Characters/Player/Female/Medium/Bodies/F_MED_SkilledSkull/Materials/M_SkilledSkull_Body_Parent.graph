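# Parent material for the SkilledSkull player body cosmetic, built on the shared 2023 character material functions for skeletal meshes.
# MF_CharacterParent_2023_PreFX builds the base material attributes, and MF_CharacterParent_2023_PostFX finishes them.
# Between them, the UseRemapCurveByFresnel static switch, on by default, inserts the skin-specific MF_SkilledSkull_RemapCurveByFresnel effect, which takes a mask texture parameter defaulting to white.
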Material /BRCosmetics/Characters/Player/Female/Medium/Bodies/F_MED_SkilledSkull/Materials/M_SkilledSkull_Body_Parent
graph {
 properties {
  bUseMaterialAttributes=true
  bUsedWithMeshDeformer=true
  bUsedWithMorphTargets=true
  bUsedWithSkeletalMesh=true
  pixelDepthOffsetMode=PDOM_Legacy
  shadingModel=MSM_FromMaterialExpression
  usedShadingModels=Default Lit | Subsurface | Clear Coat | Subsurface Profile | Cloth
 }
 nodes [
  n1 = MaterialFunctionCall() {materialFunction=&MF_CharacterParent_2023_PreFX}
  n2 = TextureObjectParameter() {
   group=RemapCurveByFresnel
   parameterName=TechArtMaskRe
   samplerType=SAMPLERTYPE_Masks
   texture=&T_WhiteMask
  }
  MaterialFunctionCall(
   Material=StaticSwitchParameter(
    True=MaterialFunctionCall!(MA=@n1.Result, FXMask=@n2, PreskinLocalPosition=@n1.Result) {materialFunction=&/MF_SkilledSkull_RemapCurveByFresnel}.Result
    False=@n1.Result
   ) {defaultValue=true, group=Skin FX, parameterName=UseRemapCurveByFresnel}
   M=@n1.M
   Pre Skinned Local Position=@n1.M
   Distance Blend=@n1.M
   Fresnel=@n1.M
  ) {materialFunction=&MF_CharacterParent_2023_PostFX}
 ]
}
