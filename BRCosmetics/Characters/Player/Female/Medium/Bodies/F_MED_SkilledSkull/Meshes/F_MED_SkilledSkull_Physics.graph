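# Physics asset for the F_MED_SkilledSkull female medium player body, covering only a dangling strap, two grenade props and the left leg.
# Capsules, plus a box on the thigh, give the left thigh, calf and calf twist bones their collision shapes.
# The strap is a chain from base_strap through five dyn_strap bones, with limited swing and nearly locked twist.
# Base strap, grenades and calf link to the thigh with free angular motion, as does calf twist to calf.
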PhysicsAsset /BRCosmetics/Characters/Player/Female/Medium/Bodies/F_MED_SkilledSkull/Meshes/F_MED_SkilledSkull_Physics
bodies {
 base_strap [{shapeType=Capsule, length=0.505, radius=0.505, shapeName=base_strap_capsule, rotation=<rot 0 -90 0>}]
 calf_l [{
   shapeType=Capsule
   length=23.5425
   radius=5.4103
   shapeName=calf_l_capsule
   center=<-13.8221 -1.959 -0.6931>
   rotation=<rot 0.7037 95.8062 -86.2386>
  }]
 calf_twist_01_l [{
   shapeType=Capsule
   length=6.2401
   radius=6.9449
   shapeName=calf_twist_01_l_capsule
   center=<-1.3759 -0.3193 0.2205>
   rotation=<rot 27.3398 -88.4581 -89.0968>
  }]
 dyn_granade_A [{
   shapeType=Capsule
   length=5.2809
   radius=4.4943
   shapeName=dyn_granade_A_capsule
   center=<-3.4077 2.0574 0.0226>
   rotation=<rot 0 1.2251 -88.3156>
  }]
 dyn_granade_B [{
   shapeType=Capsule
   length=5.2775
   radius=4.46
   shapeName=dyn_granade_B_capsule
   center=<-3.3953 1.7957 0.0039>
   rotation=<rot 0 1.3675 -89.5413>
  }]
 dyn_strap_01 [{shapeType=Capsule, length=1.9484, radius=0.8505, shapeName=dyn_strap_01_capsule, center=<2.2932 0 0>, rotation=<rot 90 0 0>}]
 dyn_strap_03 [{shapeType=Capsule, length=3, radius=0.8405, shapeName=dyn_strap_03_capsule, center=<2 0 0.1077>, rotation=<rot 90 0 0>}]
 dyn_strap_04 [{shapeType=Capsule, length=3.6225, radius=1.2029, shapeName=dyn_strap_04_capsule, center=<3 0 0.1732>, rotation=<rot 90 0 0>}]
 dyn_strap_05 [{shapeType=Capsule, length=3, radius=1.009, shapeName=dyn_strap_05_capsule, center=<3 0 0.1917>, rotation=<rot 90 0 0>}]
 dyn_strap_06 [{shapeType=Capsule, length=3, radius=1, shapeName=dyn_strap_06_capsule, center=<3 0 0.2766>, rotation=<rot 90 0 0>}]
 thigh_l [
  {
   shapeType=Capsule
   length=19.9929
   radius=6.1881
   shapeName=thigh_l_capsule
   center=<-27.5527 -0.7475 0.3422>
   rotation=<rot 0.5215 -92.9453 -87.9447>
  }
  {
   shapeType=Box
   extentX=7.5
   extentY=15.5426
   extentZ=3.2837
   shapeName=thigh_l_box
   center=<-14.6546 0.0449 -6.8944>
   rotation=<rot 0 -3.028 0>
  }
 ]
}
constraints [
 {bone1Name=dyn_strap_01, bone2Name=base_strap, swing1LimitDegrees=50, swing2LimitDegrees=30, twistLimitDegrees=1}
 {bone1Name=dyn_strap_03, bone2Name=dyn_strap_01, swing1LimitDegrees=55, swing2LimitDegrees=30, twistLimitDegrees=1}
 {bone1Name=dyn_strap_04, bone2Name=dyn_strap_03, swing1LimitDegrees=55, swing2LimitDegrees=30, twistLimitDegrees=1}
 {bone1Name=dyn_strap_05, bone2Name=dyn_strap_04, swing1LimitDegrees=55, swing2LimitDegrees=30, twistLimitDegrees=1}
 {bone1Name=dyn_strap_06, bone2Name=dyn_strap_05, swing1LimitDegrees=55, swing2LimitDegrees=30, twistLimitDegrees=1}
 {
  bone1Name=base_strap
  bone2Name=thigh_l
  swing1LimitDegrees=45
  swing1Motion=Free
  swing2LimitDegrees=45
  swing2Motion=Free
  twistLimitDegrees=45
  twistMotion=Free
 }
 {
  bone1Name=calf_l
  bone2Name=thigh_l
  swing1LimitDegrees=45
  swing1Motion=Free
  swing2LimitDegrees=45
  swing2Motion=Free
  twistLimitDegrees=45
  twistMotion=Free
 }
 {
  bone1Name=dyn_granade_A
  bone2Name=thigh_l
  swing1LimitDegrees=45
  swing1Motion=Free
  swing2LimitDegrees=45
  swing2Motion=Free
  twistLimitDegrees=45
  twistMotion=Free
 }
 {
  bone1Name=dyn_granade_B
  bone2Name=thigh_l
  swing1LimitDegrees=45
  swing1Motion=Free
  swing2LimitDegrees=45
  swing2Motion=Free
  twistLimitDegrees=45
  twistMotion=Free
 }
 {
  bone1Name=calf_twist_01_l
  bone2Name=calf_l
  swing1LimitDegrees=45
  swing1Motion=Free
  swing2LimitDegrees=45
  swing2Motion=Free
  twistLimitDegrees=45
  twistMotion=Free
 }
]
