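# Physics asset for the carabiner accessory on the female medium Skilled Skull player body, set up for dangling secondary motion.
# A box body on dyn_carabiner is constrained to a tiny capsule on base_carabiner, forming a near-hinge that swings mainly on one axis with almost no side swing or twist.
# An unconstrained capsule on thigh_r approximates the right thigh, apparently as a collision shape for the swinging piece.
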PhysicsAsset /BRCosmetics/Characters/Player/Female/Medium/Bodies/F_MED_SkilledSkull/Meshes/F_MED_SkilledSkull_carabiner_Physics
bodies {
 base_carabiner [{shapeType=Capsule, length=0.505, radius=0.505, shapeName=base_carabiner_capsule, rotation=<rot 0 -90 0>}]
 dyn_carabiner [{
   shapeType=Box
   extentX=0.7469
   extentY=4.2151
   extentZ=5.5555
   shapeName=dyn_carabiner_box
   center=<2.3498 -0.0726 -0.135>
   rotation=<rot 0 -89.903 -89.8011>
  }]
 thigh_r [{
   shapeType=Capsule
   length=14.5717
   radius=8.5064
   shapeName=thigh_r_capsule
   center=<13.5113 1.4633 0.823>
   rotation=<rot 3.1216 -96.5256 -88.0841>
  }]
}
constraints [{bone1Name=dyn_carabiner, bone2Name=base_carabiner, swing1LimitDegrees=60, swing2LimitDegrees=2, twistLimitDegrees=0}]
