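# Post-process style secondary-motion AnimBP for the F_MED_SkilledSkull female body cosmetic, copying its pose from the attached parent mesh into a Control Rig.
# Cached-pose stages then run a belt-strap constraint and offset, dynamics, and additive strap corrections.
# Dynamics chain AnimDynamics on torso canisters and grenades into two thigh-based RigidBody sims with override physics assets.
# The event graph updates a gravity override through a deprecated helper and tracks hang-glider parachute state.
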AnimBlueprint /BRCosmetics/Characters/Player/Female/Medium/Bodies/F_MED_SkilledSkull/Meshes/F_MED_SkilledSkull_AnimBP
target_skeleton=&/F_MED_SkilledSkull_Skeleton
graphs [
 AnimGraph {nodes [
   Output Pose(Result=Use cached pose 'CRCache_postDynamics'()) {layerGroup=None}
   n1 = Control Rig!(Source=Copy Pose From Mesh() {bUseAttachedParent=true}, Alpha, Is crouching:bool=false, Is Using Hang Glider:bool=false) -> (Pose) {
    alphaBoolBlend {blendOption=Linear}
    controlRigAssetReference {blueprintRigClass=&/F_MED_SkilledSkull_CtrlRig.F_MED_SkilledSkull_CtrlRig_C}
   }
   Save cached pose 'CRCache_inputPose'(Pose=@n1) {cacheName=CRCache_inputPose}
   n2 = Anim Dynamics - Bone: base_rightTorsoCanister!(
    ComponentPose=Local To Component(LocalPose=Use cached pose 'CRCache_Predynamics'())
    Alpha=0.9
    GravityOverride:Vector="0,0,-980"
    ExternalForce="0,0,0"
   ) -> (Pose) {
    angularDampingOverride=1
    angularSpringConstant=6500
    bAngularSpring=true
    bOverrideAngularDamping=true
    bUseGravityOverride=true
    gravityScale=0.35
    simulationSpace=RootRelative
    alphaBoolBlend {blendOption=Linear}
    boundBone {boneName=base_rightTorsoCanister}
    componentAppliedLinearAccClamp=<10000 10000 10000>
    componentLinearAccScale=<1 1 1>
    componentLinearVelScale=<1 1 1>
    gravityOverride {z=-980}
    physicsBodyDefinitions [{
      boundBone {boneName=base_rightTorsoCanister}
      boxExtents=<35 45 35>
      constraintSetup {angularLimitsMax=<3 15 3>, angularLimitsMin=<-3 -15 -3>, angularTarget=<1 0 0>}
      localJointOffset=<3 0 0>
     }]
    simSpaceSettings {maxAngularAcceleration=200, maxAngularVelocity=400, simSpaceAngularAlpha=0.7}
   }
   n3 = Anim Dynamics - Bone: base_leftTorsoCanister!(ComponentPose=@n2, Alpha=0.75, GravityOverride:Vector="0,0,-980", ExternalForce="0,0,0") -> (Pose) {
    angularDampingOverride=1
    angularSpringConstant=5000
    bAngularSpring=true
    bOverrideAngularDamping=true
    bUseGravityOverride=true
    gravityScale=0.35
    simulationSpace=RootRelative
    alphaBoolBlend {blendOption=Linear}
    boundBone {boneName=base_leftTorsoCanister}
    componentAppliedLinearAccClamp=<10000 10000 10000>
    componentLinearAccScale=<1 1 1>
    componentLinearVelScale=<1 1 1>
    gravityOverride {z=-980}
    physicsBodyDefinitions [{
      boundBone {boneName=base_leftTorsoCanister}
      boxExtents=<33 42 33>
      constraintSetup {angularLimitsMax=<2.5 10 2.5>, angularLimitsMin=<-2.5 -10 -2.5>, angularTarget=<1 0 0>}
      localJointOffset=<3.5 0 0>
     }]
    simSpaceSettings {maxAngularAcceleration=200, maxAngularVelocity=400, simSpaceAngularAlpha=0.7}
   }
   n4 = Anim Dynamics - Bone: base_granade_B!(ComponentPose=@n3, Alpha=0.85, GravityOverride:Vector="0,0,-980", ExternalForce="0,0,0") -> (Pose) {
    angularDampingOverride=1
    angularSpringConstant=20000
    bAngularSpring=true
    bOverrideAngularDamping=true
    bUseGravityOverride=true
    gravityScale=0.35
    simulationSpace=BoneRelative
    alphaBoolBlend {blendOption=Linear}
    boundBone {boneName=base_granade_B}
    componentLinearAccScale=<1 1 1>
    componentLinearVelScale=<1 1 1>
    gravityOverride {z=-980}
    physicsBodyDefinitions [{
      boundBone {boneName=base_granade_B}
      boxExtents=<22 32 32>
      constraintSetup {angularLimitsMax=<12 12 12>, angularLimitsMin=<-12 -12 -12>, angularTarget=<1 0 0>}
      localJointOffset=<3.5 0 0>
     }]
    relativeSpaceBone {boneName=pelvis}
    simSpaceSettings {maxAngularAcceleration=10, maxAngularVelocity=5, simSpaceAngularAlpha=0.5}
   }
   n5 = Anim Dynamics - Bone: base_granade_A!(ComponentPose=@n4, Alpha=1, GravityOverride:Vector="0,0,-980", ExternalForce="0,0,0") -> (Pose) {
    angularDampingOverride=1
    angularSpringConstant=20500
    bAngularSpring=true
    bOverrideAngularDamping=true
    bUseGravityOverride=true
    gravityScale=0.35
    simulationSpace=BoneRelative
    alphaBoolBlend {blendOption=Linear}
    boundBone {boneName=base_granade_A}
    componentLinearAccScale=<1 1 1>
    componentLinearVelScale=<1 1 1>
    gravityOverride {z=-980}
    physicsBodyDefinitions [{
      boundBone {boneName=base_granade_A}
      boxExtents=<20 30 30>
      constraintSetup {angularLimitsMax=<15 15 15>, angularLimitsMin=<-15 -15 -15>, angularTarget=<1 0 0>}
      localJointOffset=<4 0 0>
     }]
    relativeSpaceBone {boneName=pelvis}
    simSpaceSettings {maxAngularAcceleration=10, maxAngularVelocity=5, simSpaceAngularAlpha=0.5}
   }
   n6 = RigidBody!(ComponentPose=@n5, Alpha, OverrideWorldGravity:Vector="0,0,0", ExternalForce) -> (Pose) {
    bOverrideWorldGravity=true
    overridePhysicsAsset=&/F_MED_SkilledSkull_carabiner_Physics
    simulationSpace=BaseBoneSpace
    alphaBoolBlend {blendOption=Linear}
    baseBoneRef {boneName=thigh_r}
    componentAppliedLinearAccClamp=<3000 3000 3000>
    simSpaceSettings {
     dampingAlpha=0
     maxAngularAcceleration=10
     maxAngularVelocity=5
     maxLinearAcceleration=700
     maxLinearVelocity=500
     velocityScaleZ=0.5
     worldAlpha=0.7
     externalLinearDragV=<0.1 0.1 0.1>
    }
   }
   n7 = RigidBody(ComponentPose=@n6, OverrideWorldGravity="0,0,-980") {
    bOverrideWorldGravity=true
    overridePhysicsAsset=&/F_MED_SkilledSkull_Physics
    simulationSpace=BaseBoneSpace
    alphaBoolBlend {blendOption=Linear}
    baseBoneRef {boneName=thigh_l}
    componentAppliedLinearAccClamp=<3000 3000 3000>
    overrideWorldGravity {z=-980}
    simSpaceSettings {
     dampingAlpha=0
     maxAngularAcceleration=15
     maxAngularVelocity=7
     maxLinearAcceleration=900
     maxLinearVelocity=700
     velocityScaleZ=0.5
     worldAlpha=0.7
     externalLinearDragV=<0.1 0.1 0.1>
    }
   }
   Save cached pose 'CRCAche_dynamics'(Pose=Component To Local(ComponentPose=@n7)) {cacheName=CRCAche_dynamics}
   n8 = "Constraint - base_beltStrap (PPPPPP)"!(
    ComponentPose=Local To Component(LocalPose=Use cached pose 'CRCache_inputPose'())
    Alpha=1
    ConstraintWeights_0=0.514
    ConstraintWeights_1:float=0.2382
    ConstraintWeights_2:float=0.1934
    ConstraintWeights_3:float=0.0288
    ConstraintWeights_4:float=0.0137
    ConstraintWeights_5:float=0.0119
   ) -> (Pose) {
    alphaBoolBlend {blendOption=Linear}
    boneToModify {boneName=base_beltStrap}
    constraintSetup [
     {targetBone {boneName=left_belt_02}}
     {targetBone {boneName=pelvis}}
     {targetBone {boneName=thigh_twist_01_l}}
     {targetBone {boneName=deform_groin_l}}
     {targetBone {boneName=deform_glute_l}}
     {targetBone {boneName=thigh_l}}
    ]
    constraintWeights [1,1,1,1,1,1]
   }
   n9 = "Transform (Modify) Bone - Bone: base_strap"!(ComponentPose=@n8, bAlphaBoolEnabled:bool=True, Translation="0.7,0,0", Rotation="-20,-25,25") -> (Pose) {
    alphaInputType=Bool
    rotationMode=BMM_Additive
    rotationSpace=BCS_BoneSpace
    translationMode=BMM_Additive
    translationSpace=BCS_BoneSpace
    alphaBoolBlend {blendInTime=0.2, blendOption=Linear, blendOutTime=0.2}
    boneToModify {boneName=base_strap}
    rotation=<rot -20 -25 25>
    translation {x=0.7}
   }
   Save cached pose 'CRCache_Predynamics'(Pose=Component To Local(ComponentPose=@n9)) {cacheName=CRCache_Predynamics}
   n10 = "Transform (Modify) Bone - Bone: dyn_strap_02"!(
    ComponentPose=Local To Component(LocalPose=Use cached pose 'CRCAche_dynamics'())
    bAlphaBoolEnabled:bool=True
    Translation="0.066,0.281,-0.428"
    Rotation="0.207,-12.773,-0.021"
   ) -> (Pose) {
    alphaInputType=Bool
    rotationMode=BMM_Additive
    rotationSpace=BCS_BoneSpace
    translationMode=BMM_Additive
    translationSpace=BCS_BoneSpace
    alphaBoolBlend {blendInTime=0.2, blendOption=Linear, blendOutTime=0.2}
    boneToModify {boneName=dyn_strap_02}
    rotation=<rot 0.207 -12.773 -0.021>
    translation=<0.066 0.281 -0.428>
   }
   n11 = "Transform (Modify) Bone - Bone: dyn_strap_03"!(ComponentPose=@n10, bAlphaBoolEnabled:bool=True, Translation="0,0,0", Rotation="0.036,7.748,0.007") -> (Pose) {
    alphaInputType=Bool
    rotationMode=BMM_Additive
    rotationSpace=BCS_BoneSpace
    translationMode=BMM_Additive
    translationSpace=BCS_BoneSpace
    alphaBoolBlend {blendInTime=0.2, blendOption=Linear, blendOutTime=0.2}
    boneToModify {boneName=dyn_strap_03}
    rotation=<rot 0.036 7.7481 0.007>
   }
   n12 = "Transform (Modify) Bone - Bone: dyn_strap_04"!(ComponentPose=@n11, bAlphaBoolEnabled:bool=True, Translation="0,0,0", Rotation="0.010,4.374,-0.003") -> (Pose) {
    alphaInputType=Bool
    rotationMode=BMM_Additive
    rotationSpace=BCS_BoneSpace
    translationMode=BMM_Additive
    translationSpace=BCS_BoneSpace
    alphaBoolBlend {blendInTime=0.2, blendOption=Linear, blendOutTime=0.2}
    boneToModify {boneName=dyn_strap_04}
    rotation=<rot 0.01 4.374 -0.003>
   }
   Save cached pose 'CRCache_postDynamics'(Pose=Component To Local(ComponentPose=@n12)) {cacheName=CRCache_postDynamics}
  ]}
 EventGraph {nodes [
   Event BlueprintUpdateAnimation()
   > n1 = Gravity Override State Function DEPRECATED(CharacterPartAnimInstance=Self-Reference(), GameTimeInSeconds=GetGameTimeInSeconds(), GravityOverrideParams=Get Gravity Override Params())
   > Set GravityOverride(GravityOverride=@n1)
   Event OnParachuteOpened()
   > Switch on EGliderType(Selection=Get PlayerGliderType() -> (PlayerGliderType=HangGlider))
   > Set isUsingHangGlidingInParachute(isUsingHangGlidingInParachute:bool=true) -> (Output_Get:bool)
   Event OnParachuteClosed()
   > Set isUsingHangGlidingInParachute(isUsingHangGlidingInParachute=false)
  ]}
]
variables [
 Gravity Override Params {type=GravityOverrideParamsStruct}
 GravityOverride {type=Vector}
 isUsingHangGlidingInParachute {type=boolean}
]
