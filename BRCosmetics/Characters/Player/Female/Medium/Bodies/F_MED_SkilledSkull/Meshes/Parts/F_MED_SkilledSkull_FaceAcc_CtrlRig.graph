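# Control Rig for the F_MED_SkilledSkull character's face accessory that corrects hair bones as the head turns.
# DeltaRotation v4 splits head rotation relative to head_null into swing or twist, remapping one Euler axis into a clamped weight.
# Hair_strands_fix, the only function wired to execution, uses these weights to rotate select back and right strands.
# Unwired functions hold weighted parent constraints and similar fixes for three hair chains and their strands.
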ControlRigBlueprint /BRCosmetics/Characters/Player/Female/Medium/Bodies/F_MED_SkilledSkull/Meshes/Parts/F_MED_SkilledSkull_FaceAcc_CtrlRig
graphs [
 New Function_ContainedGraph {nodes [
   RigVMFunctionEntryNode()
   > RigVMFunction_Sequence = RigVMAggregateNode()
   &RigUnit_ParentConstraint(
    AdvancedSettings=(RotationOrderForFilter=XZY)
    Child=(Type=Bone,Name="dyn_hairChain_L_01")
    Parents=((Item=(Type=Bone,Name="head"),Weight=1.0))
    Weight=1
    ExecutePin=@RigVMFunction_Sequence.A
   )
   > &RigUnit_ParentConstraint(
    AdvancedSettings=(RotationOrderForFilter=XZY)
    Child=(Type=Bone,Name="dyn_hairChain_L_02")
    Parents=((Item=(Type=Bone,Name="spine_05"),Weight=0.0784),(Item=(Type=Bone,Name="neck_01"),Weight=0.1332),(Item=(Type=Bone,Name="neck_02"),Weight=0.1465),(Item=(Type=Bone,Name="head"),Weight=0.642))
    Weight=1
   )
   > &RigUnit_ParentConstraint(
    AdvancedSettings=(RotationOrderForFilter=XZY)
    Child=(Type=Bone,Name="dyn_hairChain_L_03")
    Parents=((Item=(Type=Bone,Name="clavicle_l"),Weight=0.0251),(Item=(Type=Bone,Name="neck_02"),Weight=0.2829),(Item=(Type=Bone,Name="spine_05"),Weight=0.3171),(Item=(Type=Bone,Name="neck_01"),Weight=0.3749))
    Weight=1
   )
   &RigUnit_ParentConstraint(
    AdvancedSettings=(RotationOrderForFilter=XZY)
    Child=(Type=Bone,Name="dyn_hairChain_R_01")
    Parents=((Item=(Type=Bone,Name="head"),Weight=1.0))
    Weight=1
    ExecutePin=@RigVMFunction_Sequence.B
   )
   > &RigUnit_ParentConstraint(
    AdvancedSettings=(RotationOrderForFilter=XZY)
    Child=(Type=Bone,Name="dyn_hairChain_R_02")
    Parents=((Item=(Type=Bone,Name="spine_05"),Weight=0.072),(Item=(Type=Bone,Name="neck_01"),Weight=0.127),(Item=(Type=Bone,Name="neck_02"),Weight=0.1649),(Item=(Type=Bone,Name="head"),Weight=0.636))
    Weight=1
   )
   > &RigUnit_ParentConstraint(
    AdvancedSettings=(RotationOrderForFilter=XZY)
    Child=(Type=Bone,Name="dyn_hairChain_R_03")
    Parents=((Item=(Type=Bone,Name="clavicle_r"),Weight=0.0251),(Item=(Type=Bone,Name="neck_02"),Weight=0.2829),(Item=(Type=Bone,Name="spine_05"),Weight=0.3171),(Item=(Type=Bone,Name="neck_01"),Weight=0.3749))
    Weight=1
   )
   &RigUnit_ParentConstraint(
    AdvancedSettings=(RotationOrderForFilter=XZY)
    Child=(Type=Bone,Name="dyn_hairChain_C_01")
    Parents=((Item=(Type=Bone,Name="neck_02"),Weight=0.0434),(Item=(Type=Bone,Name="head"),Weight=0.9566))
    Weight=1
    ExecutePin=@RigVMFunction_Sequence.C
   )
   > &RigUnit_ParentConstraint(
    AdvancedSettings=(RotationOrderForFilter=XZY)
    Child=(Type=Bone,Name="dyn_hairChain_C_02")
    Parents=((Item=(Type=Bone,Name="spine_05"),Weight=0.1017),(Item=(Type=Bone,Name="neck_01"),Weight=0.2392),(Item=(Type=Bone,Name="neck_02"),Weight=0.2625),(Item=(Type=Bone,Name="head"),Weight=0.3966))
    Weight=1
   )
   > &RigUnit_ParentConstraint(
    AdvancedSettings=(RotationOrderForFilter=XZY)
    Child=(Type=Bone,Name="dyn_hairChain_C_03")
    Parents=((Item=(Type=Bone,Name="clavicle_l"),Weight=0.0086),(Item=(Type=Bone,Name="neck_02"),Weight=0.1704),(Item=(Type=Bone,Name="spine_05"),Weight=0.3955),(Item=(Type=Bone,Name="neck_01"),Weight=0.4254))
    Weight=1
   )
   RigVMFunctionReturnNode(ExecuteContext=@RigVMFunction_Sequence.D)
  ]}
 New Function_ContainedGraph {nodes [
   n1 = RigVMFunctionEntryNode()
   RigVMFunction_Sequence = &RigVMFunction_Sequence(ExecuteContext=@n1.ExecuteContext)
   &RigUnit_ParentConstraint(
    AdvancedSettings=(RotationOrderForFilter=XZY)
    Child=(Type=Bone,Name="base_rightHair_D")
    Parents="((Item=(Type=Bone,Name=\"dyn_hairChain_R_03\"),Weight=0.0011),(Item=(Type=Bone,Name=\"dyn_hairChain_C_02\"),Weight=0.004),(Item=(Type=Bone,Name=\"dyn_hairChain_C_01\"),Weight=0.0098),(Item=(Type=Bone,Name=\"head\"),Weight=0.1013),(Item=(Type=Bone,Name=\"dyn_hairCha ...(truncated)"
    Weight=1
    ExecutePin=@RigVMFunction_Sequence.A
   )
   > &RigUnit_ParentConstraint(
    AdvancedSettings=(RotationOrderForFilter=XZY)
    Child=(Type=Bone,Name="base_rightHair_C")
    Parents="((Item=(Type=Bone,Name=\"dyn_hairChain_C_03\"),Weight=0.0018),(Item=(Type=Bone,Name=\"dyn_hairChain_R_03\"),Weight=0.0159),(Item=(Type=Bone,Name=\"head\"),Weight=0.024),(Item=(Type=Bone,Name=\"dyn_hairChain_C_01\"),Weight=0.0251),(Item=(Type=Bone,Name=\"dyn_hairCha ...(truncated)"
    Weight=1
   )
   > &RigUnit_ParentConstraint(
    AdvancedSettings=(RotationOrderForFilter=XZY)
    Child=(Type=Bone,Name="base_backHair_J")
    Parents="((Item=(Type=Bone,Name=\"dyn_hairChain_C_03\"),Weight=0.0005),(Item=(Type=Bone,Name=\"dyn_hairChain_R_03\"),Weight=0.0011),(Item=(Type=Bone,Name=\"dyn_hairChain_R_02\"),Weight=0.1256),(Item=(Type=Bone,Name=\"dyn_hairChain_C_02\"),Weight=0.1539),(Item=(Type=Bone,Na ...(truncated)"
    Weight=1
   )
   > &RigUnit_ParentConstraint(
    AdvancedSettings=(RotationOrderForFilter=XZY)
    Child=(Type=Bone,Name="base_backHair_H")
    Parents="((Item=(Type=Bone,Name=\"dyn_hairChain_L_02\"),Weight=0.001),(Item=(Type=Bone,Name=\"dyn_hairChain_R_03\"),Weight=0.0036),(Item=(Type=Bone,Name=\"dyn_hairChain_R_01\"),Weight=0.0065),(Item=(Type=Bone,Name=\"head\"),Weight=0.0163),(Item=(Type=Bone,Name=\"dyn_hairCha ...(truncated)"
    Weight=1
   )
   > &RigUnit_ParentConstraint(
    AdvancedSettings=(RotationOrderForFilter=XZY)
    Child=(Type=Bone,Name="base_backHair_F")
    Parents="((Item=(Type=Bone,Name=\"dyn_hairChain_C_03\"),Weight=0.0053),(Item=(Type=Bone,Name=\"dyn_hairChain_L_03\"),Weight=0.0116),(Item=(Type=Bone,Name=\"head\"),Weight=0.0446),(Item=(Type=Bone,Name=\"dyn_hairChain_C_01\"),Weight=0.1532),(Item=(Type=Bone,Name=\"dyn_hairCh ...(truncated)"
    Weight=1
   )
   > &RigUnit_ParentConstraint(
    AdvancedSettings=(RotationOrderForFilter=XZY)
    Child=(Type=Bone,Name="base_backHair_D")
    Parents="((Item=(Type=Bone,Name=\"dyn_hairChain_L_03\"),Weight=0.0004),(Item=(Type=Bone,Name=\"dyn_hairChain_C_02\"),Weight=0.0355),(Item=(Type=Bone,Name=\"dyn_hairChain_L_02\"),Weight=0.1256),(Item=(Type=Bone,Name=\"dyn_hairChain_C_01\"),Weight=0.133),(Item=(Type=Bone,Nam ...(truncated)"
    Weight=1
   )
   > &RigUnit_ParentConstraint(
    AdvancedSettings=(RotationOrderForFilter=XZY)
    Child=(Type=Bone,Name="dyn_backHair_E")
    Parents="((Item=(Type=Bone,Name=\"dyn_hairChain_C_03\"),Weight=0.0033),(Item=(Type=Bone,Name=\"head\"),Weight=0.0109),(Item=(Type=Bone,Name=\"dyn_hairChain_C_01\"),Weight=0.011),(Item=(Type=Bone,Name=\"dyn_hairChain_C_02\"),Weight=0.0489),(Item=(Type=Bone,Name=\"dyn_hairCha ...(truncated)"
    Weight=1
   )
   > &RigUnit_ParentConstraint(
    AdvancedSettings=(RotationOrderForFilter=XZY)
    Child=(Type=Bone,Name="dyn_leftHair_I")
    Parents="((Item=(Type=Bone,Name=\"dyn_hairChain_C_01\"),Weight=0.0004),(Item=(Type=Bone,Name=\"dyn_hairChain_C_02\"),Weight=0.0004),(Item=(Type=Bone,Name=\"dyn_hairChain_L_03\"),Weight=0.0486),(Item=(Type=Bone,Name=\"head\"),Weight=0.0841),(Item=(Type=Bone,Name=\"dyn_hairCh ...(truncated)"
    Weight=1
   )
   RigVMFunctionReturnNode(ExecuteContext=@RigVMFunction_Sequence.B)
  ]}
 New Function_ContainedGraph {nodes [
   RigVMFunctionEntryNode()
   > RigVMFunction_Sequence = RigVMAggregateNode()
   DeltaRotation v4 = RigVMFunctionReferenceNode(
    Rotation Order=ZYX
    child=(Type=Bone,Name="head")
    debugRotation=false
    outputMinMax=(X=0,Y=1)
    parent=(Type=Null,Name="head_null")
    premult=1
    rotationMinMax=(X=5,Y=-20)
    triggerAxis=Z
    twistAxis=(X=1,Y=0,Z=0)
    ExecuteContext=@RigVMFunction_Sequence.A
   )
   > &RigUnit_ModifyTransforms(
    ItemToModify="((Item=(Type=Bone,Name=\"dyn_hairChain_C_01\"),Transform=(Rotation=(X=0,Y=0,Z=0,W=1),Translation=(X=0,Y=0,Z=0),Scale3D=(X=1,Y=1,Z=1))),(Item=(Type=Bone,Name=\"dyn_hairChain_C_02\"),Transform=(Rotation=(X=0,Y=0,Z=0.1719,W=-0.9851),Translation=(X=0.2713,Y=0,Z=0) ...(truncated)"
    WeightMinimum=0
    Weight=@"DeltaRotation v4".outputWeight
   )
   > &RigUnit_ModifyTransforms(
    ItemToModify="((Item=(Type=Bone,Name=\"dyn_hairChain_L_01\"),Transform=(Rotation=(X=0.1883,Y=-0.0368,Z=0.1882,W=-0.9632),Translation=(X=0,Y=0,Z=0),Scale3D=(X=1,Y=1,Z=1))),(Item=(Type=Bone,Name=\"dyn_hairChain_L_02\"),Transform=(Rotation=(X=0,Y=0,Z=0,W=1),Translation=(X=0,Y= ...(truncated)"
    WeightMinimum=0
    Weight=@"DeltaRotation v4".outputWeight
   )
   > &RigUnit_ModifyTransforms(
    ItemToModify="((Item=(Type=Bone,Name=\"dyn_hairChain_R_01\"),Transform=(Rotation=(X=0,Y=0,Z=0,W=1),Translation=(X=0,Y=0,Z=0),Scale3D=(X=1,Y=1,Z=1))),(Item=(Type=Bone,Name=\"dyn_hairChain_R_02\"),Transform=(Rotation=(X=-0.0817,Y=0.2704,Z=-0.1015,W=0.9539),Translation=(X=0,Y= ...(truncated)"
    WeightMinimum=0
    Weight=@"DeltaRotation v4".outputWeight
   )
   DeltaRotation v4_1 = RigVMFunctionReferenceNode(
    Rotation Order=ZYX
    child=(Type=Bone,Name="head")
    debugRotation=false
    onlyTwist=true
    outputMinMax=(X=0,Y=1)
    parent=(Type=Null,Name="head_null")
    premult=1
    rotationMinMax=(X=5,Y=-50)
    triggerAxis=X
    twistAxis=(X=1,Y=0,Z=0)
    ExecuteContext=@RigVMFunction_Sequence.B
   )
   > &RigUnit_ModifyTransforms(
    ItemToModify="((Item=(Type=Bone,Name=\"dyn_hairChain_C_01\"),Transform=(Rotation=(X=0,Y=0,Z=0,W=1),Translation=(X=0,Y=0.6069,Z=0),Scale3D=(X=1,Y=1,Z=1))),(Item=(Type=Bone,Name=\"dyn_hairChain_C_02\"),Transform=(Rotation=(X=0,Y=0,Z=0,W=1),Translation=(X=0,Y=0.2935,Z=-1.6061) ...(truncated)"
    WeightMinimum=0
    Weight=@"DeltaRotation v4_1".outputWeight
   )
   > &RigUnit_ModifyTransforms(
    ItemToModify="((Item=(Type=Bone,Name=\"dyn_hairChain_L_01\"),Transform=(Rotation=(X=0,Y=0,Z=0,W=1),Translation=(X=0,Y=0,Z=0),Scale3D=(X=1,Y=1,Z=1))),(Item=(Type=Bone,Name=\"dyn_hairChain_L_02\"),Transform=(Rotation=(X=0,Y=0,Z=0,W=1),Translation=(X=0,Y=0,Z=0),Scale3D=(X=1,Y= ...(truncated)"
    WeightMinimum=0
    Weight=@"DeltaRotation v4_1".outputWeight
   )
   > &RigUnit_ModifyTransforms(
    ItemToModify="((Item=(Type=Bone,Name=\"dyn_hairChain_R_01\"),Transform=(Rotation=(X=0,Y=0,Z=0,W=1),Translation=(X=0,Y=0,Z=0),Scale3D=(X=1,Y=1,Z=1))),(Item=(Type=Bone,Name=\"dyn_hairChain_R_02\"),Transform=(Rotation=(X=0,Y=0,Z=0,W=1),Translation=(X=0,Y=0,Z=0),Scale3D=(X=1,Y= ...(truncated)"
    WeightMinimum=0
    Weight=@"DeltaRotation v4_1".outputWeight
   )
   DeltaRotation v4_2 = RigVMFunctionReferenceNode(
    Rotation Order=ZYX
    child=(Type=Bone,Name="head")
    debugRotation=false
    onlyTwist=true
    outputMinMax=(X=0,Y=1)
    parent=(Type=Null,Name="head_null")
    premult=1
    rotationMinMax=(X=5,Y=55)
    triggerAxis=X
    twistAxis=(X=1,Y=0,Z=0)
    ExecuteContext=@RigVMFunction_Sequence.C
   )
   > &RigUnit_ModifyTransforms(
    ItemToModify="((Item=(Type=Bone,Name=\"dyn_hairChain_C_01\"),Transform=(Rotation=(X=0.6722,Y=-0.0395,Z=0.0329,W=-0.7386),Translation=(X=0.1132,Y=1.1912,Z=0.1122),Scale3D=(X=1,Y=1,Z=1))),(Item=(Type=Bone,Name=\"dyn_hairChain_C_02\"),Transform=(Rotation=(X=0,Y=0,Z=0,W=1),Tran ...(truncated)"
    WeightMinimum=0
    Weight=@"DeltaRotation v4_2".outputWeight
   )
   > &RigUnit_ModifyTransforms(
    ItemToModify="((Item=(Type=Bone,Name=\"dyn_hairChain_L_01\"),Transform=(Rotation=(X=-0.4533,Y=0,Z=0,W=0.8914),Translation=(X=0,Y=0,Z=0),Scale3D=(X=1,Y=1,Z=1))),(Item=(Type=Bone,Name=\"dyn_hairChain_L_02\"),Transform=(Rotation=(X=0,Y=0,Z=0,W=1),Translation=(X=0,Y=0,Z=0),Scal ...(truncated)"
    WeightMinimum=0
    Weight=@"DeltaRotation v4_2".outputWeight
   )
   > &RigUnit_ModifyTransforms(
    ItemToModify="((Item=(Type=Bone,Name=\"dyn_hairChain_R_01\"),Transform=(Rotation=(X=0.8696,Y=0,Z=0,W=-0.4938),Translation=(X=0,Y=0,Z=0),Scale3D=(X=1,Y=1,Z=1))),(Item=(Type=Bone,Name=\"dyn_hairChain_R_02\"),Transform=(Rotation=(X=0,Y=0,Z=0,W=1),Translation=(X=0,Y=0,Z=0),Scal ...(truncated)"
    WeightMinimum=0
    Weight=@"DeltaRotation v4_2".outputWeight
   )
   RigVMFunctionReturnNode(ExecuteContext=@RigVMFunction_Sequence.D)
   &RigUnit_ModifyTransforms(
    ExecutePin=()
    ItemToModify="((Item=(Type=Bone,Name=\"dyn_hairChain_C_01\"),Transform=(Rotation=(X=0,Y=0,Z=0,W=1),Translation=(X=0,Y=0,Z=0),Scale3D=(X=1,Y=1,Z=1))),(Item=(Type=Bone,Name=\"dyn_hairChain_C_02\"),Transform=(Rotation=(X=0,Y=0,Z=0,W=1),Translation=(X=0,Y=0,Z=0),Scale3D=(X=1,Y= ...(truncated)"
    Weight=1
    WeightMinimum=0
   )
   > &RigUnit_ModifyTransforms(
    ItemToModify="((Item=(Type=Bone,Name=\"dyn_hairChain_L_01\"),Transform=(Rotation=(X=0,Y=0,Z=0,W=1),Translation=(X=0,Y=0,Z=0),Scale3D=(X=1,Y=1,Z=1))),(Item=(Type=Bone,Name=\"dyn_hairChain_L_02\"),Transform=(Rotation=(X=0,Y=0,Z=0,W=1),Translation=(X=0,Y=0,Z=0),Scale3D=(X=1,Y= ...(truncated)"
    Weight=1
    WeightMinimum=0
   )
   > &RigUnit_ModifyTransforms(
    ItemToModify="((Item=(Type=Bone,Name=\"dyn_hairChain_R_01\"),Transform=(Rotation=(X=0,Y=0,Z=0,W=1),Translation=(X=0,Y=0,Z=0),Scale3D=(X=1,Y=1,Z=1))),(Item=(Type=Bone,Name=\"dyn_hairChain_R_02\"),Transform=(Rotation=(X=0,Y=0,Z=0,W=1),Translation=(X=0,Y=0,Z=0),Scale3D=(X=1,Y= ...(truncated)"
    Weight=1
    WeightMinimum=0
   )
  ]}
 New Function_ContainedGraph {nodes [
   n1 = RigVMFunctionEntryNode()
   RigVMFunction_MathBoolToInteger = &RigVMFunction_MathBoolToInteger(Value=@n1.onlyTwist)
   RigVMFunction_MathQuaternionSwingTwist = &RigVMFunction_MathQuaternionSwingTwist(
    TwistAxis=@n1.twistAxis
    Input=&RigUnit_GetRelativeTransformForItem(bChildInitial=False, bParentInitial=False, Child=@n1.child, Parent=@n1.parent).RelativeTransform.Rotation
   )
   n2 = RigVMDispatchNode(
    Values=((),())
    Index=@RigVMFunction_MathBoolToInteger.Result
    Values.0=@RigVMFunction_MathQuaternionSwingTwist.Swing
    Values.1=@RigVMFunction_MathQuaternionSwingTwist.Twist
   ) {template="DISPATCH_RigVMDispatch_SelectInt32(in Index,in Values,out Result)"}
   RigVMFunction_MathQuaternionToEuler = &RigVMFunction_MathQuaternionToEuler(Value=@n2.Result, RotationOrder=@n1.Rotation Order)
   > DISPATCH_RigVMDispatch_Print = RigVMDispatchNode(ScreenColor=(R=1,G=1,B=1,A=1), ScreenDuration=1.5, Enabled=@n1.debugRotation, Value=@RigVMFunction_MathQuaternionToEuler.Result) {template="DISPATCH_RigVMDispatch_Print(in Prefix,in Value,in Enabled,in ScreenDuration,in ScreenColor)"}
   Cast = RigVMDispatchNode(Value=@n1.triggerAxis) {template="DISPATCH_RigVMDispatch_CastEnumToInt(in Value,out Result)"}
   DISPATCH_RigVMDispatch_SelectInt32 = RigVMDispatchNode(
    Values=(0,0,0)
    Index=@Cast.Result
    Values.0=@RigVMFunction_MathQuaternionToEuler.Result.X
    Values.1=@RigVMFunction_MathQuaternionToEuler.Result.Y
    Values.2=@RigVMFunction_MathQuaternionToEuler.Result.Z
   ) {template="DISPATCH_RigVMDispatch_SelectInt32(in Index,in Values,out Result)"}
   Remap = &RigVMFunction_MathDoubleRemap(
    bClamp=False
    Value=@DISPATCH_RigVMDispatch_SelectInt32.Result
    TargetMinimum=@n1.outputMinMax.X
    TargetMaximum=@n1.outputMinMax.Y
    SourceMinimum=@n1.rotationMinMax.X
    SourceMaximum=@n1.rotationMinMax.Y
   )
   Absolute = &RigVMFunction_MathDoubleAbs(Value=@Remap.Result)
   Multiply = &RigVMFunction_MathDoubleMul(
    A=&RigVMFunction_MathDoubleClamp(Value=RigVMDispatchNode(True=@Absolute.Result, Condition=@n1.absoluteValue, False=@Remap.Result) {template="DISPATCH_RigVMDispatch_If(in Condition,in True,in False,out Result)"}.Result).Result
    B=@n1.premult
   )
   > RigVMFunctionReturnNode(outputQuat=@n2.Result, outputWeight=@Multiply.Result)
  ]}
 New Function_ContainedGraph {nodes [
   RigVMFunctionEntryNode()
   > RigVMFunction_Sequence = RigVMAggregateNode()
   DeltaRotation v4 = RigVMFunctionReferenceNode(
    Rotation Order=ZYX
    child=(Type=Bone,Name="head")
    debugRotation=false
    outputMinMax=(X=0,Y=1)
    parent=(Type=Null,Name="head_null")
    premult=1
    rotationMinMax=(X=5,Y=-20)
    triggerAxis=Z
    twistAxis=(X=1,Y=0,Z=0)
    ExecuteContext=@RigVMFunction_Sequence.A
   )
   > &RigUnit_ModifyTransforms(
    ItemToModify=((Item=(Type=Bone,Name="dyn_backHair_H"),Transform=(Rotation=(X=0,Y=0,Z=-0.1086,W=0.9941),Translation=(X=0,Y=0,Z=0),Scale3D=(X=1,Y=1,Z=1))))
    WeightMinimum=0
    Weight=@"DeltaRotation v4".outputWeight
   )
   DeltaRotation v4_1 = RigVMFunctionReferenceNode(
    Rotation Order=ZYX
    child=(Type=Bone,Name="head")
    debugRotation=false
    onlyTwist=true
    outputMinMax=(X=0,Y=1)
    parent=(Type=Null,Name="head_null")
    premult=1
    rotationMinMax=(X=5,Y=-50)
    triggerAxis=X
    twistAxis=(X=1,Y=0,Z=0)
    ExecuteContext=@RigVMFunction_Sequence.B
   )
   > &RigUnit_ModifyTransforms(
    ItemToModify=((Item=(Type=Bone,Name="dyn_backHair_H"),Transform=(Rotation=(X=0,Y=0.392,Z=0,W=-0.92),Translation=(X=0,Y=0,Z=0),Scale3D=(X=1,Y=1,Z=1))))
    WeightMinimum=0
    Weight=@"DeltaRotation v4_1".outputWeight
   )
   > &RigUnit_ModifyTransforms(
    ItemToModify=((Item=(Type=Bone,Name="dyn_backHair_J"),Transform=(Rotation=(X=0,Y=0.392,Z=0,W=-0.92),Translation=(X=0,Y=0,Z=0),Scale3D=(X=1,Y=1,Z=1))))
    WeightMinimum=0
    Weight=@"DeltaRotation v4_1".outputWeight
   )
   > &RigUnit_ModifyTransforms(
    ItemToModify=((Item=(Type=Bone,Name="dyn_rightHair_E"),Transform=(Rotation=(X=0,Y=-0.3065,Z=0,W=0.9519),Translation=(X=0,Y=0,Z=0),Scale3D=(X=1,Y=1,Z=1))))
    WeightMinimum=0
    Weight=@"DeltaRotation v4_1".outputWeight
   )
   DeltaRotation v4_2 = RigVMFunctionReferenceNode(
    Rotation Order=ZYX
    child=(Type=Bone,Name="head")
    debugRotation=false
    onlyTwist=true
    outputMinMax=(X=0,Y=1)
    parent=(Type=Null,Name="head_null")
    premult=1
    rotationMinMax=(X=5,Y=50)
    triggerAxis=X
    twistAxis=(X=1,Y=0,Z=0)
    ExecuteContext=@RigVMFunction_Sequence.C
   )
   > &RigUnit_ModifyTransforms(
    ItemToModify=((Item=(Type=Bone,Name="dyn_rightHair_E"),Transform=(Rotation=(X=0,Y=0.8662,Z=0,W=0.4997),Translation=(X=0,Y=0,Z=0),Scale3D=(X=1,Y=1,Z=1))))
    WeightMinimum=0
    Weight=@"DeltaRotation v4_2".outputWeight
   )
   RigVMFunctionReturnNode(ExecuteContext=@RigVMFunction_Sequence.D)
  ]}
 RigVMFunctionLibrary {nodes [
   DeltaRotation v4 = RigVMCollapseNode(
    ExecuteContext=()
    Rotation Order=ZYX
    child=(Type=Bone,Name="None")
    outputMinMax=(X=0,Y=1)
    parent=(Type=Bone,Name="None")
    premult=1
    rotationMinMax=(X=0,Y=0)
    twistAxis=(X=1,Y=0,Z=0)
   )
   Hair_chains_contraints = RigVMCollapseNode(ExecuteContext=())
   Hair_chains_fix = RigVMCollapseNode(ExecuteContext=())
   Hair_strands_constraints = RigVMCollapseNode(ExecuteContext=())
   Hair_strands_fix = RigVMCollapseNode(ExecuteContext=())
  ]}
 RigVMFunction_Sequence_1_ContainedGraph {nodes [
   n1 = RigVMFunctionEntryNode()
   RigVMFunction_Sequence = &RigVMFunction_Sequence(ExecuteContext=@n1.ExecuteContext)
   n2 = &RigVMFunction_Sequence(ExecuteContext=@RigVMFunction_Sequence.B)
   n3 = &RigVMFunction_Sequence(ExecuteContext=@n2.B)
   RigVMFunctionReturnNode(A=@RigVMFunction_Sequence.A, B=@n2.A, C=@n3.A, D=@n3.B)
  ]}
 RigVMFunction_Sequence_1_ContainedGraph {nodes [
   n1 = RigVMFunctionEntryNode()
   RigVMFunction_Sequence = &RigVMFunction_Sequence(ExecuteContext=@n1.ExecuteContext)
   n2 = &RigVMFunction_Sequence(ExecuteContext=@RigVMFunction_Sequence.B)
   n3 = &RigVMFunction_Sequence(ExecuteContext=@n2.B)
   RigVMFunctionReturnNode(A=@RigVMFunction_Sequence.A, B=@n2.A, C=@n3.A, D=@n3.B)
  ]}
 RigVMFunction_Sequence_1_ContainedGraph {nodes [
   n1 = RigVMFunctionEntryNode()
   RigVMFunction_Sequence = &RigVMFunction_Sequence(ExecuteContext=@n1.ExecuteContext)
   n2 = &RigVMFunction_Sequence(ExecuteContext=@RigVMFunction_Sequence.B)
   n3 = &RigVMFunction_Sequence(ExecuteContext=@n2.B)
   RigVMFunctionReturnNode(A=@RigVMFunction_Sequence.A, B=@n2.A, C=@n3.A, D=@n3.B)
  ]}
 RigVMModel {nodes [
   RigUnit_BeginExecution = &RigUnit_BeginExecution()
   Hair_strands_fix = RigVMFunctionReferenceNode(ExecuteContext=@RigUnit_BeginExecution.ExecutePin)
   Hair_chains_contraints = RigVMFunctionReferenceNode(ExecuteContext=())
   Hair_chains_fix = RigVMFunctionReferenceNode(ExecuteContext=())
   Hair_strands_constraints = RigVMFunctionReferenceNode(ExecuteContext=())
  ]}
]
hierarchy [
 root {loc=<0 0 0>, rot=<0 0 0 1>}
 pelvis {parent=root, loc=<0 -0.5301 98.0571>, rot=<0.0127 0.707 -0.0127 -0.707>}
 spine_01 {parent=pelvis, loc=<6.1208 -0.9328 0>, rot=<0 0 0.034 -0.9994>}
 spine_02 {parent=spine_01, loc=<6.5997 -0.2095 0>, rot=<0 0 0.0238 -0.9997>}
 spine_03 {parent=spine_02, loc=<6.4914 -0.1237 0>, rot=<0 0 0.0149 -0.9999>}
 spine_04 {parent=spine_03, loc=<6.9598 0.0794 0>, rot=<0 0 -0.0132 -0.9999>}
 spine_05 {parent=spine_04, loc=<6.9987 0 0>, rot=<0 0 -0.0015 -1>}
 neck_01 {parent=spine_05, loc=<11.3138 0.1529 0>, rot=<0 0 -0.2003 -0.9797>}
 neck_02 {parent=neck_01, loc=<4.5627 0 0>, rot=<0 0 0.0504 -0.9987>}
 head {parent=neck_02, loc=<5.3914 0 0>, rot=<0 0 0.115 -0.9934>}
 base_frontHair_A {parent=head, loc=<7.6065 10.4524 0.3367>, rot=<0.1885 -0.1113 0.9745 0.0495>}
 dyn_frontHair_A_01 {parent=base_frontHair_A, loc=<0 0 0>, rot=<0 0 0 1>}
 dyn_frontHair_A_02 {parent=dyn_frontHair_A_01, loc=<1.8865 0 0>, rot=<0 0.0742 0 -0.9972>}
 base_frontHair_B {parent=head, loc=<8.6062 9.3651 3.2986>, rot=<0.2478 -0.3121 0.9136 0.081>}
 dyn_frontHair_B_01 {parent=base_frontHair_B, loc=<0 0 0>, rot=<0 0 0 1>}
 dyn_frontHair_B_02 {parent=dyn_frontHair_B_01, loc=<1.6462 0 0>, rot=<0 0.2694 0 -0.963>}
 base_frontHair_C {parent=head, loc=<11.9848 9.7377 3.2405>, rot=<-0.1479 0.1335 -0.7011 -0.6847>}
 dyn_frontHair_C {parent=base_frontHair_C, loc=<0 0 0>, rot=<0 0 0 1>}
 dyn_frontHair_C_01 {parent=base_frontHair_C, loc=<0 0 0>, rot=<0 0 0 1>}
 dyn_frontHair_C_02 {parent=dyn_frontHair_C_01, loc=<2.549 0 0>, rot=<0.0031 0.0136 -0.222 -0.9749>}
 dyn_frontHair_C_03 {parent=dyn_frontHair_C_02, loc=<2.0455 0 0>, rot=<0.1605 -0.0185 0.003 -0.9869>}
 base_frontHair_D {parent=head, loc=<9.1922 7.2197 7.3238>, rot=<-0.2895 0.4129 -0.6826 -0.5289>}
 dyn_frontHair_D {parent=base_frontHair_D, loc=<0 0 0>, rot=<0 0 0 1>}
 base_rightHair_A {parent=head, loc=<7.9845 0.3747 9.2032>, rot=<0.5479 -0.6467 0.3854 0.3649>}
 dyn_rightHair_A {parent=base_rightHair_A, loc=<0 0 0>, rot=<0 0 0 1>}
 base_rightHair_B {parent=head, loc=<1.7095 -7.8144 5.3513>, rot=<0.3126 -0.9199 0.2278 -0.0641>}
 dyn_rightHair_B {parent=base_rightHair_B, loc=<0 0 0>, rot=<0 0 0 1>}
 base_rightHair_C {parent=head, loc=<-4.2633 -6.0778 4.6988>, rot=<0.4015 -0.8699 0.2539 0.1328>}
 dyn_rightHair_C {parent=base_rightHair_C, loc=<0 0 0>, rot=<0 0 0 1>}
 base_rightHair_D {parent=head, loc=<-0.7513 -4.0552 6.121>, rot=<0.3885 -0.7809 0.4533 0.1838>}
 dyn_rightHair_D {parent=base_rightHair_D, loc=<0 0 0>, rot=<0 0 0 1>}
 base_rightHair_E {parent=head, loc=<-3.8708 -0.6247 5.8314>, rot=<0.13 -0.878 0.4401 0.1363>}
 dyn_rightHair_E {parent=base_rightHair_E, loc=<0 0 0>, rot=<0 0 0 1>}
 base_rightHair_F {parent=head, loc=<7.3182 5.5517 7.2079>, rot=<0.0639 0.132 0.9583 -0.2453>}
 dyn_rightHair_F {parent=base_rightHair_F, loc=<0 0 0>, rot=<0 0 0 1>}
 base_rightHair_G {parent=head, loc=<4.1079 1.0319 9.7905>, rot=<0.3617 -0.7198 0.5839 0.1007>}
 dyn_rightHair_G {parent=base_rightHair_G, loc=<0 0 0>, rot=<0 0 0 1>}
 base_rightHair_H {parent=head, loc=<1.5316 -1.7916 9.1905>, rot=<0.0141 -0.6819 0.7062 -0.1897>}
 dyn_rightHair_H {parent=base_rightHair_H, loc=<0 0 0>, rot=<0 0 0 1>}
 base_rightHair_I {parent=head, loc=<0.9918 1.0132 7.6999>, rot=<-0.1713 -0.707 0.6852 -0.0365>}
 dyn_rightHair_I {parent=base_rightHair_I, loc=<0 0 0>, rot=<0 0 0 1>}
 base_backHair_A {parent=head, loc=<14.9128 -3.487 -2.0759>, rot=<0.203 -0.4154 0.2011 -0.8636>}
 dyn_backHair_A {parent=base_backHair_A, loc=<0 0 0>, rot=<0 0 0 1>}
 base_backHair_B {parent=head, loc=<5.2517 -7.8709 -4.1353>, rot=<0.4285 -0.8005 -0.1943 -0.3712>}
 dyn_backHair_B {parent=base_backHair_B, loc=<0 0 0>, rot=<0 0 0 1>}
 base_backHair_C {parent=head, loc=<2.136 -6.194 -5.9775>, rot=<0.3926 -0.8745 -0.2416 -0.1506>}
 dyn_backHair_C {parent=base_backHair_C, loc=<0 0 0>, rot=<0 0 0 1>}
 base_backHair_D {parent=head, loc=<-0.3865 -7.5444 -4.0159>, rot=<0.2635 -0.9328 -0.1916 -0.1539>}
 dyn_backHair_D {parent=base_backHair_D, loc=<0 0 0>, rot=<0 0 0 1>}
 base_backHair_E {parent=head, loc=<-3.812 -5.5253 -3.7466>, rot=<0.1999 -0.9585 -0.1758 -0.1018>}
 dyn_backHair_E {parent=base_backHair_E, loc=<0 0 0>, rot=<0 0 0 1>}
 base_backHair_F {parent=head, loc=<-2.649 -6.8552 -2.5932>, rot=<0.3602 -0.9214 -0.1458 -0.0103>}
 dyn_backHair_F {parent=base_backHair_F, loc=<0 0 0>, rot=<0 0 0 1>}
 base_backHair_G {parent=head, loc=<3.1125 -9.999 0.1405>, rot=<0.0762 -0.9517 -0.2342 0.1835>}
 dyn_backHair_G {parent=base_backHair_G, loc=<0 0 0>, rot=<0 0 0 1>}
 base_backHair_H {parent=head, loc=<-4.2705 -7.9257 0.185>, rot=<0.2712 -0.9464 0.1738 0.0244>}
 dyn_backHair_H {parent=base_backHair_H, loc=<0 0 0>, rot=<0 0 0 1>}
 base_backHair_I {parent=head, loc=<5.1626 -8.2554 3.2838>, rot=<0.2529 -0.9064 0.3098 0.1358>}
 dyn_backHair_I {parent=base_backHair_I, loc=<0 0 0>, rot=<0 0 0 1>}
 base_backHair_J {parent=head, loc=<-1.3914 -7.766 2.4991>, rot=<0.4753 -0.8744 0.0209 0.0953>}
 dyn_backHair_J {parent=base_backHair_J, loc=<0 0 0>, rot=<0 0 0 1>}
 base_backHair_K {parent=head, loc=<10.1437 -8.618 1.9174>, rot=<0.666 -0.7452 0.0325 -0.0063>}
 dyn_backHair_K {parent=base_backHair_K, loc=<0 0 0>, rot=<0 0 0 1>}
 base_leftHair_A {parent=head, loc=<10.5575 7.3257 -6.503>, rot=<0.3194 -0.2838 -0.6393 -0.6393>}
 dyn_leftHair_A {parent=base_leftHair_A, loc=<0 0 0>, rot=<0 0 0 1>}
 base_leftHair_B {parent=head, loc=<6.1217 7.5451 -5.9584>, rot=<0.1495 0.5021 0.8518 0.0015>}
 dyn_leftHair_B_01 {parent=base_leftHair_B, loc=<0 0 0>, rot=<0 0 0 1>}
 dyn_leftHair_B_02 {parent=dyn_leftHair_B_01, loc=<2.3516 0 0>, rot=<0.0016 0.0296 -0.0542 -0.9981>}
 base_leftHair_C {parent=head, loc=<5.9461 6.1651 -8.0389>, rot=<-0.2605 0.5432 0.7238 0.3364>}
 dyn_leftHair_C {parent=base_leftHair_C, loc=<0 0 0>, rot=<0 0 0 1>}
 base_leftHair_D {parent=head, loc=<2.4312 4.1155 -8.5085>, rot=<-0.1612 -0.709 -0.6856 0.0353>}
 dyn_leftHair_D {parent=base_leftHair_D, loc=<0 0 0>, rot=<0 0 0 1>}
 base_leftHair_E {parent=head, loc=<2.1837 3.5155 -9.2773>, rot=<0.0674 -0.7497 -0.65 0.1048>}
 dyn_leftHair_E {parent=base_leftHair_E, loc=<0 0 0>, rot=<0 0 0 1>}
 base_leftHair_F {parent=head, loc=<-0.8875 1.1179 -6.7457>, rot=<0.0302 0.6618 0.748 -0.04>}
 dyn_leftHair_F_01 {parent=base_leftHair_F, loc=<0 0 0>, rot=<0 0 0 1>}
 dyn_leftHair_F_02 {parent=dyn_leftHair_F_01, loc=<2.3279 0 0>, rot=<0.0233 -0.0215 0.2417 -0.9698>}
 base_leftHair_G {parent=head, loc=<4.1139 1.5168 -9.1363>, rot=<-0.3563 0.555 0.6948 0.287>}
 dyn_leftHair_G {parent=base_leftHair_G, loc=<0 0 0>, rot=<0 0 0 1>}
 base_leftHair_H {parent=head, loc=<0.9599 -3.4598 -7.0018>, rot=<0.2663 -0.8809 -0.3885 -0.0475>}
 dyn_leftHair_H {parent=base_leftHair_H, loc=<0 0 0>, rot=<0 0 -0.0469 -0.9989>}
 base_leftHair_I {parent=head, loc=<-2.7995 -3.043 -5.6963>, rot=<0.2131 -0.8059 -0.4994 -0.2359>}
 dyn_leftHair_I {parent=base_leftHair_I, loc=<0 0 0>, rot=<0 0 0 1>}
 base_leftHair_J {parent=head, loc=<9.374 -1.1675 -8.8769>, rot=<0.3421 -0.6378 -0.6004 -0.3401>}
 dyn_leftHair_J {parent=base_leftHair_J, loc=<0 0 0>, rot=<0 0 0 1>}
 FACIAL_C_Jawline {parent=head, loc=<-5.9934 5.6787 0>, rot=<0.0029 -0.7071 -0.0029 -0.7071>}
 base_backhair_L {parent=head, loc=<-4.356 -6.7552 2.6373>, rot=<0.3725 -0.9057 0.1459 0.1402>}
 dyn_backHair_L_01 {parent=base_backhair_L, loc=<0 0 0>, rot=<0 0 0 1>}
 dyn_backHair_L_02 {parent=dyn_backHair_L_01, loc=<3.732 0 0>, rot=<0 0 0.053 -0.9986>}
 dyn_backHair_L_03 {parent=dyn_backHair_L_02, loc=<4.0408 0 0>, rot=<0 0 0 1>}
 base_backhair_R {parent=head, loc=<1.0044 -3.4526 4.7116>, rot=<0.0142 0.4007 0.9126 -0.0796>}
 dyn_backHair_R_01 {parent=base_backhair_R, loc=<0 0 0>, rot=<0.5074 -0.0365 0.0232 -0.8606>}
 dyn_backHair_R_02 {parent=dyn_backHair_R_01, loc=<3.7321 0 0>, rot=<0 0 0.053 -0.9986>}
 dyn_backHair_R_03 {parent=dyn_backHair_R_02, loc=<4.0405 0 0>, rot=<0 0 0 1>}
 base_backhair_C1 {parent=head, loc=<1.0299 -6.5374 0.0763>, rot=<-0.0042 -0.0857 0.9947 -0.0569>}
 dyn_backHair_C_01 {parent=base_backhair_C1, loc=<0 0 0>, rot=<0 0 0 1>}
 dyn_backHair_C_02 {parent=dyn_backHair_C_01, loc=<3.7321 0 0>, rot=<0 0 0.053 -0.9986>}
 dyn_backHair_C_03 {parent=dyn_backHair_C_02, loc=<4.6756 0 0>, rot=<0 0 0 1>}
 base_hairChain_L {parent=head, loc=<1.0043 -3.4526 -4.7116>, rot=<0.0041 -1 0 0>}
 dyn_hairChain_L_01 {parent=base_hairChain_L, loc=<0 0 0>, rot=<0 0 0 1>}
 dyn_hairChain_L_02 {parent=dyn_hairChain_L_01, loc=<3.7274 0 0>, rot=<0 0 0 1>}
 dyn_hairChain_L_03 {parent=dyn_hairChain_L_02, loc=<4.0777 0 0>, rot=<0 0 0 1>}
 base_hairChain_R {parent=head, loc=<1.0044 -3.4526 4.7116>, rot=<0.0041 -1 0 0>}
 dyn_hairChain_R_01 {parent=base_hairChain_R, loc=<0 0 0>, rot=<0 0 0 1>}
 dyn_hairChain_R_02 {parent=dyn_hairChain_R_01, loc=<3.7274 0 0>, rot=<0 0 0 1>}
 dyn_hairChain_R_03 {parent=dyn_hairChain_R_02, loc=<4.0777 0 0>, rot=<0 0 0 1>}
 base_hairChain_C {parent=head, loc=<1.0245 -5.8846 0.0763>, rot=<0.0041 -1 0 0>}
 dyn_hairChain_C_01 {parent=base_hairChain_C, loc=<0 0 0>, rot=<0 0 0 1>}
 dyn_hairChain_C_02 {parent=dyn_hairChain_C_01, loc=<3.7274 0 0>, rot=<0 0 0 1>}
 dyn_hairChain_C_03 {parent=dyn_hairChain_C_02, loc=<4.6673 0 0>, rot=<0 0 0 1>}
 head_null {parent=neck_02, type=Null, loc=<5.3914 0 0>, rot=<0 0 0.115 -0.9934>}
 clavicle_l {parent=spine_05, loc=<5.4831 -0.1053 -3.2946>, rot=<0.0465 -0.7289 -0.009 -0.6829>}
 upperarm_l {parent=clavicle_l, loc=<9.8658 -1.2136 -0.4105>, rot=<0.0164 -0.3245 -0.0269 -0.9454>}
 lowerarm_l {parent=upperarm_l, loc=<26.465 -0.8345 0.259>, rot=<0.0036 0.0161 -0.2298 -0.9731>}
 hand_l {parent=lowerarm_l, loc=<22.7355 -0.3203 -0.6105>, rot=<-0.6784 -0.0223 0.0735 -0.7307>}
 thumb_01_l {parent=hand_l, loc=<1.3118 -0.3191 -2.459>, rot=<0.6546 -0.4042 0.0664 -0.6354>}
 thumb_02_l {parent=thumb_01_l, loc=<5.9336 -0.0448 0.0401>, rot=<-0.0037 0.0213 0.298 -0.9543>}
 thumb_03_l {parent=thumb_02_l, loc=<2.2223 0.3682 -0.0907>, rot=<0.0118 -0.013 0.0066 -0.9998>}
 index_metacarpal_l {parent=hand_l, loc=<1.6535 0.0183 -2.96>, rot=<0.0158 -0.0124 -0.0173 -0.9996>}
 index_01_l {parent=index_metacarpal_l, loc=<8.3008 0.0357 -0.0524>, rot=<-0.0044 0.0485 0.0924 -0.9945>}
 index_02_l {parent=index_01_l, loc=<3.3587 -0.3004 0.0452>, rot=<-0.002 0.0027 0.3225 -0.9466>}
 index_03_l {parent=index_02_l, loc=<2.5578 -0.123 -0.0252>, rot=<-0.002 0.0067 0.2558 -0.9667>}
 middle_metacarpal_l {parent=hand_l, loc=<2.1856 -0.1725 -0.5984>, rot=<0.006 0.0168 -0.0329 -0.9993>}
 middle_01_l {parent=middle_metacarpal_l, loc=<7.3275 -0.2447 0.0129>, rot=<-0.0208 0.0042 0.0858 -0.9961>}
 middle_02_l {parent=middle_01_l, loc=<4.0123 0.1401 0.1175>, rot=<-0.0218 0.0258 0.3512 -0.9357>}
 middle_03_l {parent=middle_02_l, loc=<2.6077 0.0534 0.1373>, rot=<0.0082 0.0188 0.1708 -0.9851>}
 ring_metacarpal_l {parent=hand_l, loc=<2.2074 -0.3625 1.3638>, rot=<-0.0284 0.0722 -0.0458 -0.9959>}
 ring_01_l {parent=ring_metacarpal_l, loc=<7.0295 -0.6197 -0.1446>, rot=<-0.0489 -0.0252 0.1892 -0.9804>}
 ring_02_l {parent=ring_01_l, loc=<3.6174 0.4647 -0.1027>, rot=<0 0 0.2806 -0.9598>}
 ring_03_l {parent=ring_02_l, loc=<2.4152 0.114 0.0199>, rot=<0 0 0.1189 -0.9929>}
 pinky_metacarpal_l {parent=hand_l, loc=<1.7616 -0.9753 3.3814>, rot=<-0.0397 0.0814 -0.069 -0.9935>}
 pinky_01_l {parent=pinky_metacarpal_l, loc=<7.1709 -0.4831 -0.1094>, rot=<-0.0654 -0.0035 0.2242 -0.9723>}
 pinky_02_l {parent=pinky_01_l, loc=<2.8274 0.0893 -0.1043>, rot=<0.0037 0.0011 0.2854 -0.9584>}
 pinky_03_l {parent=pinky_02_l, loc=<2.2046 0.0307 -0.0078>, rot=<-0.0002 0.0098 0.1445 -0.9895>}
 deform_wrist_dn_base_l {parent=hand_l, loc=<0 -1.5326 0>, rot=<0 0 0 1>}
 deform_wrist_dn_l {parent=deform_wrist_dn_base_l, loc=<0 0 0>, rot=<0 0 0 1>}
 deform_wrist_up_base_l {parent=hand_l, loc=<0 2.3009 0>, rot=<0 0 0 1>}
 deform_wrist_up_l {parent=deform_wrist_up_base_l, loc=<0 0 0>, rot=<0 0 0 1>}
 lowerarm_twist_02_l {parent=lowerarm_l, loc=<10.6707 0 0>, rot=<0 0 0 1>}
 lowerarm_twist_01_l {parent=lowerarm_l, loc=<18.316 0 0>, rot=<0 0 0 1>}
 deform_elbow_in_base_l {parent=lowerarm_l, loc=<0.1834 3.6062 -0.2572>, rot=<0 0 0 1>}
 deform_elbow_in_l {parent=deform_elbow_in_base_l, loc=<0 0 0>, rot=<0 0 0 1>}
 deform_elbow_out_base_l {parent=lowerarm_l, loc=<0 -4 0>, rot=<0 0 0 1>}
 deform_elbow_out_l {parent=deform_elbow_out_base_l, loc=<0 0 0>, rot=<0 0 0 1>}
 upperarm_twist_01_l {parent=upperarm_l, loc=<7.2135 -0.1134 0>, rot=<0 0 0 1>}
 deform_upperarm_bk_base_l {parent=upperarm_twist_01_l, loc=<-6.4053 -4.3746 0.314>, rot=<0 0 0 1>}
 deform_upperarm_bk_l {parent=deform_upperarm_bk_base_l, loc=<0 0 0>, rot=<0 0 0 1>}
 deform_upperarm_fr_base_l {parent=upperarm_twist_01_l, loc=<-5.6312 5.092 0.3354>, rot=<0 0 0 1>}
 deform_upperarm_fr_l {parent=deform_upperarm_fr_base_l, loc=<0 0 0>, rot=<0 0 0 1>}
 deform_upperarm_up_base_l {parent=upperarm_twist_01_l, loc=<-8.891 0.4825 3.9517>, rot=<0 0 0 1>}
 deform_upperarm_up_l {parent=deform_upperarm_up_base_l, loc=<0 0 0>, rot=<0 0 0 1>}
 upperarm_twist_02_l {parent=upperarm_l, loc=<19.7998 -0.1134 0>, rot=<0 0 0 1>}
 clavicle_r {parent=spine_05, loc=<5.4831 -0.1053 3.2946>, rot=<0.7289 0.0465 -0.6829 0.009>}
 upperarm_r {parent=clavicle_r, loc=<-9.8658 1.2136 0.4105>, rot=<0.0164 -0.3245 -0.0269 -0.9454>}
 lowerarm_r {parent=upperarm_r, loc=<-26.465 0.8345 -0.259>, rot=<0.0036 0.0161 -0.2298 -0.9731>}
 hand_r {parent=lowerarm_r, loc=<-22.7355 0.3203 0.6106>, rot=<-0.6784 -0.0223 0.0735 -0.7307>}
 thumb_01_r {parent=hand_r, loc=<-1.3118 0.3191 2.459>, rot=<0.6546 -0.4042 0.0664 -0.6354>}
 thumb_02_r {parent=thumb_01_r, loc=<-5.9336 0.0448 -0.0401>, rot=<-0.0037 0.0213 0.298 -0.9543>}
 thumb_03_r {parent=thumb_02_r, loc=<-2.2223 -0.3682 0.0907>, rot=<0.0118 -0.013 0.0066 -0.9998>}
 index_metacarpal_r {parent=hand_r, loc=<-1.6535 -0.0183 2.96>, rot=<0.0158 -0.0124 -0.0173 -0.9996>}
 index_01_r {parent=index_metacarpal_r, loc=<-8.3008 -0.0357 0.0524>, rot=<-0.0044 0.0485 0.0924 -0.9945>}
 index_02_r {parent=index_01_r, loc=<-3.3587 0.3004 -0.0452>, rot=<-0.002 0.0027 0.3225 -0.9466>}
 index_03_r {parent=index_02_r, loc=<-2.5578 0.123 0.0252>, rot=<-0.002 0.0067 0.2558 -0.9667>}
 middle_metacarpal_r {parent=hand_r, loc=<-2.1856 0.1725 0.5984>, rot=<0.006 0.0168 -0.0329 -0.9993>}
 middle_01_r {parent=middle_metacarpal_r, loc=<-7.3275 0.2447 -0.0129>, rot=<-0.0208 0.0042 0.0858 -0.9961>}
 middle_02_r {parent=middle_01_r, loc=<-4.0123 -0.1401 -0.1175>, rot=<-0.0218 0.0258 0.3512 -0.9357>}
 middle_03_r {parent=middle_02_r, loc=<-2.6077 -0.0534 -0.1373>, rot=<0.0082 0.0188 0.1708 -0.9851>}
 ring_metacarpal_r {parent=hand_r, loc=<-2.2074 0.3625 -1.3638>, rot=<-0.0284 0.0722 -0.0458 -0.9959>}
 ring_01_r {parent=ring_metacarpal_r, loc=<-7.0295 0.6197 0.1446>, rot=<-0.0489 -0.0252 0.1892 -0.9804>}
 ring_02_r {parent=ring_01_r, loc=<-3.6174 -0.4647 0.1027>, rot=<0 0 0.2806 -0.9598>}
 ring_03_r {parent=ring_02_r, loc=<-2.4152 -0.114 -0.0199>, rot=<0 0 0.1189 -0.9929>}
 pinky_metacarpal_r {parent=hand_r, loc=<-1.7616 0.9753 -3.3814>, rot=<-0.0397 0.0814 -0.069 -0.9935>}
 pinky_01_r {parent=pinky_metacarpal_r, loc=<-7.1709 0.4831 0.1094>, rot=<-0.0654 -0.0035 0.2242 -0.9723>}
 pinky_02_r {parent=pinky_01_r, loc=<-2.8274 -0.0893 0.1043>, rot=<0.0037 0.0011 0.2854 -0.9584>}
 pinky_03_r {parent=pinky_02_r, loc=<-2.2046 -0.0307 0.0078>, rot=<-0.0002 0.0098 0.1445 -0.9895>}
 deform_wrist_dn_base_r {parent=hand_r, loc=<-0.0001 1.5328 0>, rot=<0 0 0 1>}
 deform_wrist_dn_r {parent=deform_wrist_dn_base_r, loc=<0 0 0>, rot=<0 0 0 1>}
 deform_wrist_up_base_r {parent=hand_r, loc=<-0.0002 -2.3007 -0.0001>, rot=<0 0 0 1>}
 deform_wrist_up_r {parent=deform_wrist_up_base_r, loc=<0 0 0>, rot=<0 0 0 1>}
 lowerarm_twist_02_r {parent=lowerarm_r, loc=<-10.6707 0 0>, rot=<0 0 0 -1>}
 lowerarm_twist_01_r {parent=lowerarm_r, loc=<-18.316 0 0>, rot=<0 0 0 -1>}
 deform_elbow_in_base_r {parent=lowerarm_r, loc=<-0.1834 -3.6062 0.2572>, rot=<0 0 0 1>}
 deform_elbow_in_r {parent=deform_elbow_in_base_r, loc=<0 0 0>, rot=<0 0 0 1>}
 deform_elbow_out_base_r {parent=lowerarm_r, loc=<0 4 -0.0001>, rot=<0 0 0 -1>}
 deform_elbow_out_r {parent=deform_elbow_out_base_r, loc=<0 0 0>, rot=<0 0 0 1>}
 upperarm_twist_01_r {parent=upperarm_r, loc=<-7.2135 0.1134 0>, rot=<0 0 0 -1>}
 deform_upperarm_bk_base_r {parent=upperarm_twist_01_r, loc=<6.4053 4.3746 -0.314>, rot=<0 0 0 1>}
 deform_upperarm_bk_r {parent=deform_upperarm_bk_base_r, loc=<0 0 0>, rot=<0 0 0 1>}
 deform_upperarm_fr_base_r {parent=upperarm_twist_01_r, loc=<5.6312 -5.092 -0.3354>, rot=<0 0 0 1>}
 deform_upperarm_fr_r {parent=deform_upperarm_fr_base_r, loc=<0 0 0>, rot=<0 0 0 1>}
 deform_upperarm_up_base_r {parent=upperarm_twist_01_r, loc=<8.891 -0.4825 -3.9517>, rot=<0 0 0 1>}
 deform_upperarm_up_r {parent=deform_upperarm_up_base_r, loc=<0 0 0>, rot=<0 0 0 1>}
 upperarm_twist_02_r {parent=upperarm_r, loc=<-19.7998 0.1134 0>, rot=<0 0 0 -1>}
 deform_pec_aim_base_l {parent=spine_04, loc=<3.2456 9.3309 -1.2237>, rot=<0.0603 -0.4033 0.3318 -0.8507>}
 deform_pec_aim_l {parent=deform_pec_aim_base_l, loc=<0 0 0>, rot=<0 0 0 1>}
 deform_pec_base_l {parent=deform_pec_aim_l, loc=<9.3665 0 0>, rot=<0 0 0 1>}
 deform_pec_l {parent=deform_pec_base_l, loc=<0 0 0>, rot=<0 0 0 1>}
 deform_pec_aim_base_r {parent=spine_04, loc=<3.2452 9.3309 1.2237>, rot=<-0.4033 -0.0603 0.8507 0.3318>}
 deform_pec_aim_r {parent=deform_pec_aim_base_r, loc=<0 0 0>, rot=<0 0 0 1>}
 deform_pec_base_r {parent=deform_pec_aim_r, loc=<-9.3667 -0.0003 -0.0004>, rot=<0 0 0 1>}
 deform_pec_r {parent=deform_pec_base_r, loc=<0 0 0>, rot=<0 0 0 1>}
 thigh_l {parent=pelvis, loc=<-1.0855 -0.2109 -8.7053>, rot=<0.0009 0.0355 0.0155 -0.9993>}
 calf_l {parent=thigh_l, loc=<-42.5284 -0.0019 -0.7297>, rot=<0.0006 -0.0161 -0.0391 -0.9991>}
 foot_l {parent=calf_l, loc=<-46.7401 -2.5014 -0.592>, rot=<0.0427 -0.0238 0.0353 -0.9982>}
 ball_l {parent=foot_l, loc=<-3.5658 10.8066 -0.1544>, rot=<0.0001 -0.0001 -0.7063 -0.7079>}
 calf_twist_02_l {parent=calf_l, loc=<-19.1546 -1.7943 -0.2638>, rot=<0.0237 0.0095 -0.0035 -0.9997>}
 calf_twist_01_l {parent=calf_l, loc=<-35.8483 -1.9197 -0.5794>, rot=<0.0237 0.0095 -0.0035 -0.9997>}
 deform_knee_in_base_l {parent=calf_l, loc=<0.8031 -5.0231 0.2456>, rot=<0 0 0 -1>}
 deform_knee_in_l {parent=deform_knee_in_base_l, loc=<0 0 0>, rot=<0 0 0 1>}
 deform_knee_out_base_l {parent=calf_l, loc=<0 4 0>, rot=<0 0 0 -1>}
 deform_knee_out_l {parent=deform_knee_out_base_l, loc=<0 0 0>, rot=<0 0 0 1>}
 thigh_twist_01_l {parent=thigh_l, loc=<-27.952 0 0>, rot=<0.0013 0 0 -1>}
 deform_groin_base_l {parent=thigh_l, loc=<0.4801 5.6776 -0.2486>, rot=<0 0 0 -1>}
 deform_groin_l {parent=deform_groin_base_l, loc=<0 0 0>, rot=<0 0 0 1>}
 thigh_r {parent=pelvis, loc=<-1.0855 -0.2109 8.7053>, rot=<0.0355 -0.0009 0.9993 0.0155>}
 calf_r {parent=thigh_r, loc=<42.5284 0.0019 0.7297>, rot=<0.0006 -0.0161 -0.0391 -0.9991>}
 foot_r {parent=calf_r, loc=<46.7401 2.5014 0.592>, rot=<0.0427 -0.0238 0.0353 -0.9982>}
 ball_r {parent=foot_r, loc=<3.5658 -10.8066 0.1544>, rot=<0.0001 -0.0001 -0.7063 -0.7079>}
 calf_twist_02_r {parent=calf_r, loc=<19.1546 1.7943 0.2638>, rot=<0.0237 0.0095 -0.0035 -0.9997>}
 calf_twist_01_r {parent=calf_r, loc=<35.8483 1.9197 0.5794>, rot=<0.0237 0.0095 -0.0035 -0.9997>}
 deform_knee_in_base_r {parent=calf_r, loc=<-0.8031 5.0231 -0.2457>, rot=<0 0 0 -1>}
 deform_knee_in_r {parent=deform_knee_in_base_r, loc=<0 0 0>, rot=<0 0 0 1>}
 deform_knee_out_base_r {parent=calf_r, loc=<0 -4 0>, rot=<0 0 0 -1>}
 deform_knee_out_r {parent=deform_knee_out_base_r, loc=<0 0 0>, rot=<0 0 0 1>}
 thigh_twist_01_r {parent=thigh_r, loc=<27.952 0 0>, rot=<0.0013 0 0 -1>}
 deform_groin_base_r {parent=thigh_r, loc=<-0.4801 -5.6776 0.2486>, rot=<0 0 0 -1>}
 deform_groin_r {parent=deform_groin_base_r, loc=<0 0 0>, rot=<0 0 0 1>}
 deform_glute_base_l {parent=pelvis, loc=<-7.3466 -9.0213 -9.1245>, rot=<0.0172 0.031 -0.4477 -0.8935>}
 deform_glute_l {parent=deform_glute_base_l, loc=<0 0 0>, rot=<0 0 0 1>}
 deform_glute_base_r {parent=pelvis, loc=<-7.3464 -9.0213 9.1245>, rot=<0.031 -0.0172 0.8935 -0.4477>}
 deform_glute_r {parent=deform_glute_base_r, loc=<0 0 0>, rot=<0 0 0 1>}
]
